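# Control Rig for the PhoneCharger backpack cosmetic that poses the eight legs of two spiders, A and B.
# Chained two-bone IK nodes in four sequence branches aim dynControlRig leg bones at dyn_ tip bones, using pole-vector controls parented to those tips.
# Exceptions: A leg05 solves its own dyn_ chain, A legs 06-07 have partial weight, B leg01 uses a direction pole.
# The IK-control reads and SpiderB main-control transform are unconnected to execution.
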ControlRigBlueprint /BRCosmetics/Accessories/FORT_Backpacks/Backpack_F_MED_PhoneCharger/Meshes/F_MED_PhoneCharger_Pack_CtrlRig
graphs [
 RigVMFunctionLibrary {links [], nodes []}
 RigVMFunction_Sequence_1_ContainedGraph {nodes [
   n1 = RigVMFunctionEntryNode()
   RigVMFunction_Sequence = &RigVMFunction_Sequence(ExecuteContext=@n1.ExecuteContext)
   n2 = &RigVMFunction_Sequence(ExecuteContext=@RigVMFunction_Sequence.B)
   n3 = &RigVMFunction_Sequence(ExecuteContext=@n2.B)
   n4 = &RigVMFunction_Sequence(ExecuteContext=@n3.B)
   RigVMFunctionReturnNode(A=@RigVMFunction_Sequence.A, B=@n2.A, C=@n3.A, D=@n4.A, E=@n4.B)
  ]}
 RigVMModel {nodes [
   RigUnit_BeginExecution = &RigUnit_BeginExecution()
   RigVMFunction_Sequence = RigVMAggregateNode(ExecuteContext=@RigUnit_BeginExecution.ExecutePin)
   /Script/ControlRig.RigUnit_TwoBoneIKSimplePerItem(
    DebugSettings=(bEnabled=False,Scale=10,WorldOffset=(Rotation=(X=0,Y=0,Z=0,W=1),Translation=(X=0,Y=0,Z=0),Scale3D=(X=1,Y=1,Z=1)))
    EffectorItem=(Type=Bone,Name="dyn_spiderB_leg04_03")
    ItemA=(Type=Bone,Name="dynControlRig_spiderB_leg04_01")
    ItemALength=0
    ItemB=(Type=Bone,Name="dynControlRig_spiderB_leg04_02")
    ItemBLength=0
    PoleVectorKind=Location
    PoleVectorSpace=(Type=Control,Name="None")
    PrimaryAxis=(X=1,Y=0,Z=0)
    SecondaryAxis=(X=0,Y=0,Z=1)
    SecondaryAxisWeight=1
    StretchMaximumRatio=1.25
    StretchStartRatio=0.75
    Weight=1
    bEnableStretch=False
    bPropagateToChildren=True
    PoleVector=&RigUnit_GetTransform(Item=(Type=Control,Name="dyn_spiderB_leg04_poleVector"), bInitial=False).Transform.Translation
    Effector=&RigUnit_GetTransform(Item=(Type=Bone,Name="dyn_spiderB_leg04_03"), bInitial=False).Transform
    ExecutePin=@RigVMFunction_Sequence.B
   ) {is=struct RigVMUnitNode}
   > &RigUnit_TwoBoneIKSimplePerItem(
    DebugSettings=(bEnabled=False,Scale=10,WorldOffset=(Rotation=(X=0,Y=0,Z=0,W=1),Translation=(X=0,Y=0,Z=0),Scale3D=(X=1,Y=1,Z=1)))
    EffectorItem=(Type=Bone,Name="dyn_spiderB_leg03_03")
    ItemA=(Type=Bone,Name="dynControlRig_spiderB_leg03_01")
    ItemALength=0
    ItemB=(Type=Bone,Name="dynControlRig_spiderB_leg03_02")
    ItemBLength=0
    PoleVectorKind=Location
    PoleVectorSpace=(Type=Control,Name="None")
    PrimaryAxis=(X=1,Y=0,Z=0)
    SecondaryAxis=(X=0,Y=0,Z=1)
    SecondaryAxisWeight=1
    StretchMaximumRatio=1.25
    StretchStartRatio=0.75
    Weight=1
    bEnableStretch=False
    bPropagateToChildren=True
    Effector=&RigUnit_GetTransform(Item=(Type=Bone,Name="dyn_spiderB_leg03_03"), bInitial=False).Transform
    PoleVector=&RigUnit_GetTransform(Item=(Type=Control,Name="dyn_spiderB_leg03_poleVector"), bInitial=False).Transform.Translation
   )
   > &RigUnit_TwoBoneIKSimplePerItem(
    DebugSettings=(bEnabled=False,Scale=10,WorldOffset=(Rotation=(X=0,Y=0,Z=0,W=1),Translation=(X=0,Y=0,Z=0),Scale3D=(X=1,Y=1,Z=1)))
    EffectorItem=(Type=Bone,Name="dyn_spiderB_leg02_03")
    ItemA=(Type=Bone,Name="dynControlRig_spiderB_leg02_01")
    ItemALength=0
    ItemB=(Type=Bone,Name="dynControlRig_spiderB_leg02_02")
    ItemBLength=0
    PoleVectorKind=Location
    PoleVectorSpace=(Type=Control,Name="None")
    PrimaryAxis=(X=1,Y=0,Z=0)
    SecondaryAxis=(X=0,Y=0,Z=1)
    SecondaryAxisWeight=1
    StretchMaximumRatio=1.25
    StretchStartRatio=0.75
    Weight=1
    bEnableStretch=False
    bPropagateToChildren=True
    Effector=&RigUnit_GetTransform(Item=(Type=Bone,Name="dyn_spiderB_leg02_03"), bInitial=False).Transform
    PoleVector=&RigUnit_GetTransform(Item=(Type=Control,Name="dyn_spiderB_leg02_poleVector"), bInitial=False).Transform.Translation
   )
   > &RigUnit_TwoBoneIKSimplePerItem(
    DebugSettings=(bEnabled=False,Scale=10,WorldOffset=(Rotation=(X=0,Y=0,Z=0,W=1),Translation=(X=0,Y=0,Z=0),Scale3D=(X=1,Y=1,Z=1)))
    EffectorItem=(Type=Bone,Name="dyn_spiderB_leg01_03")
    ItemA=(Type=Bone,Name="dynControlRig_spiderB_leg01_01")
    ItemALength=0
    ItemB=(Type=Bone,Name="dynControlRig_spiderB_leg01_02")
    ItemBLength=0
    PoleVectorKind=Direction
    PoleVectorSpace=(Type=Bone,Name="None")
    PrimaryAxis=(X=1,Y=0,Z=0)
    SecondaryAxis=(X=0,Y=0,Z=1)
    SecondaryAxisWeight=1
    StretchMaximumRatio=1.25
    StretchStartRatio=0.75
    Weight=1
    bEnableStretch=False
    bPropagateToChildren=True
    Effector=&RigUnit_GetTransform(Item=(Type=Bone,Name="dyn_spiderB_leg01_03"), bInitial=False).Transform
    PoleVector=&RigUnit_GetTransform(Item=(Type=Control,Name="dyn_spiderB_leg01_poleVector"), bInitial=False).Transform.Translation
   )
   &RigUnit_TwoBoneIKSimplePerItem(
    DebugSettings=(bEnabled=False,Scale=10,WorldOffset=(Rotation=(X=0,Y=0,Z=0,W=1),Translation=(X=0,Y=0,Z=0),Scale3D=(X=1,Y=1,Z=1)))
    EffectorItem=(Type=Bone,Name="dyn_spiderB_leg05_03")
    ItemA=(Type=Bone,Name="dynControlRig_spiderB_leg05_01")
    ItemALength=0
    ItemB=(Type=Bone,Name="dynControlRig_spiderB_leg05_02")
    ItemBLength=0
    PoleVectorKind=Location
    PoleVectorSpace=(Type=Control,Name="None")
    PrimaryAxis=(X=1,Y=0,Z=0)
    SecondaryAxis=(X=0,Y=0,Z=1)
    SecondaryAxisWeight=1
    StretchMaximumRatio=1.25
    StretchStartRatio=0.75
    Weight=1
    bEnableStretch=False
    bPropagateToChildren=True
    Effector=&RigUnit_GetTransform(Item=(Type=Bone,Name="dyn_spiderB_leg05_03"), bInitial=False).Transform
    PoleVector=&RigUnit_GetTransform(Item=(Type=Control,Name="dyn_spiderB_leg05_poleVector"), bInitial=False).Transform.Translation
    ExecutePin=@RigVMFunction_Sequence.C
   )
   > &RigUnit_TwoBoneIKSimplePerItem(
    DebugSettings=(bEnabled=False,Scale=10,WorldOffset=(Rotation=(X=0,Y=0,Z=0,W=1),Translation=(X=0,Y=0,Z=0),Scale3D=(X=1,Y=1,Z=1)))
    EffectorItem=(Type=Bone,Name="dyn_spiderB_leg06_03")
    ItemA=(Type=Bone,Name="dynControlRig_spiderB_leg06_01")
    ItemALength=0
    ItemB=(Type=Bone,Name="dynControlRig_spiderB_leg06_02")
    ItemBLength=0
    PoleVectorKind=Location
    PoleVectorSpace=(Type=Control,Name="None")
    PrimaryAxis=(X=1,Y=0,Z=0)
    SecondaryAxis=(X=0,Y=0,Z=1)
    SecondaryAxisWeight=1
    StretchMaximumRatio=1.25
    StretchStartRatio=0.75
    Weight=1
    bEnableStretch=False
    bPropagateToChildren=True
    PoleVector=&RigUnit_GetTransform(Item=(Type=Control,Name="dyn_spiderB_leg06_poleVector"), bInitial=False).Transform.Translation
    Effector=&RigUnit_GetTransform(Item=(Type=Bone,Name="dyn_spiderB_leg06_03"), bInitial=False).Transform
   )
   > &RigUnit_TwoBoneIKSimplePerItem(
    DebugSettings=(bEnabled=False,Scale=10,WorldOffset=(Rotation=(X=0,Y=0,Z=0,W=1),Translation=(X=0,Y=0,Z=0),Scale3D=(X=1,Y=1,Z=1)))
    EffectorItem=(Type=Bone,Name="dyn_spiderB_leg07_03")
    ItemA=(Type=Bone,Name="dynControlRig_spiderB_leg07_01")
    ItemALength=0
    ItemB=(Type=Bone,Name="dynControlRig_spiderB_leg07_02")
    ItemBLength=0
    PoleVectorKind=Location
    PoleVectorSpace=(Type=Control,Name="None")
    PrimaryAxis=(X=1,Y=0,Z=0)
    SecondaryAxis=(X=0,Y=0,Z=1)
    SecondaryAxisWeight=1
    StretchMaximumRatio=1.25
    StretchStartRatio=0.75
    Weight=1
    bEnableStretch=False
    bPropagateToChildren=True
    PoleVector=&RigUnit_GetTransform(Item=(Type=Control,Name="dyn_spiderB_leg07_poleVector"), bInitial=False).Transform.Translation
    Effector=&RigUnit_GetTransform(Item=(Type=Bone,Name="dyn_spiderB_leg07_03"), bInitial=False).Transform
   )
   > &RigUnit_TwoBoneIKSimplePerItem(
    DebugSettings=(bEnabled=False,Scale=10,WorldOffset=(Rotation=(X=0,Y=0,Z=0,W=1),Translation=(X=0,Y=0,Z=0),Scale3D=(X=1,Y=1,Z=1)))
    EffectorItem=(Type=Bone,Name="dyn_spiderB_leg08_03")
    ItemA=(Type=Bone,Name="dynControlRig_spiderB_leg08_01")
    ItemALength=0
    ItemB=(Type=Bone,Name="dynControlRig_spiderB_leg08_02")
    ItemBLength=0
    PoleVectorKind=Location
    PoleVectorSpace=(Type=Control,Name="None")
    PrimaryAxis=(X=1,Y=0,Z=0)
    SecondaryAxis=(X=0,Y=0,Z=1)
    SecondaryAxisWeight=1
    StretchMaximumRatio=1.25
    StretchStartRatio=0.75
    Weight=1
    bEnableStretch=False
    bPropagateToChildren=True
    Effector=&RigUnit_GetTransform(Item=(Type=Bone,Name="dyn_spiderB_leg08_03"), bInitial=False).Transform
    PoleVector=&RigUnit_GetTransform(Item=(Type=Control,Name="dyn_spiderB_leg08_poleVector"), bInitial=False).Transform.Translation
   )
   &RigUnit_TwoBoneIKSimplePerItem(
    DebugSettings=(bEnabled=False,Scale=10,WorldOffset=(Rotation=(X=0,Y=0,Z=0,W=1),Translation=(X=0,Y=0,Z=0),Scale3D=(X=1,Y=1,Z=1)))
    EffectorItem=(Type=Bone,Name="dyn_spiderA_leg04_03")
    ItemA=(Type=Bone,Name="dynControlRig_spiderA_leg04_01")
    ItemALength=0
    ItemB=(Type=Bone,Name="dynControlRig_spiderA_leg04_02")
    ItemBLength=0
    PoleVectorKind=Location
    PoleVectorSpace=(Type=Control,Name="None")
    PrimaryAxis=(X=1,Y=0,Z=0)
    SecondaryAxis=(X=0,Y=0,Z=1)
    SecondaryAxisWeight=1
    StretchMaximumRatio=1.25
    StretchStartRatio=0.75
    Weight=1
    bEnableStretch=False
    bPropagateToChildren=True
    Effector=&RigUnit_GetTransform(Item=(Type=Bone,Name="dyn_spiderA_leg04_03"), bInitial=False).Transform
    PoleVector=&RigUnit_GetTransform(Item=(Type=Control,Name="dyn_spiderA_leg04_poleVector"), bInitial=False).Transform.Translation
    ExecutePin=@RigVMFunction_Sequence.D
   )
   > &RigUnit_TwoBoneIKSimplePerItem(
    DebugSettings=(bEnabled=False,Scale=10,WorldOffset=(Rotation=(X=0,Y=0,Z=0,W=1),Translation=(X=0,Y=0,Z=0),Scale3D=(X=1,Y=1,Z=1)))
    EffectorItem=(Type=Bone,Name="dyn_spiderA_leg03_03")
    ItemA=(Type=Bone,Name="dynControlRig_spiderA_leg03_01")
    ItemALength=0
    ItemB=(Type=Bone,Name="dynControlRig_spiderA_leg03_02")
    ItemBLength=0
    PoleVectorKind=Location
    PoleVectorSpace=(Type=Control,Name="None")
    PrimaryAxis=(X=1,Y=0,Z=0)
    SecondaryAxis=(X=0,Y=0,Z=1)
    SecondaryAxisWeight=1
    StretchMaximumRatio=1.25
    StretchStartRatio=0.75
    Weight=1
    bEnableStretch=False
    bPropagateToChildren=True
    PoleVector=&RigUnit_GetTransform(Item=(Type=Control,Name="dyn_spiderA_leg03_poleVector"), bInitial=False).Transform.Translation
    Effector=&RigUnit_GetTransform(Item=(Type=Bone,Name="dyn_spiderA_leg03_03"), bInitial=False).Transform
   )
   > &RigUnit_TwoBoneIKSimplePerItem(
    DebugSettings=(bEnabled=False,Scale=10,WorldOffset=(Rotation=(X=0,Y=0,Z=0,W=1),Translation=(X=0,Y=0,Z=0),Scale3D=(X=1,Y=1,Z=1)))
    EffectorItem=(Type=Bone,Name="dyn_spiderA_leg02_03")
    ItemA=(Type=Bone,Name="dynControlRig_spiderA_leg02_01")
    ItemALength=0
    ItemB=(Type=Bone,Name="dynControlRig_spiderA_leg02_02")
    ItemBLength=0
    PoleVectorKind=Location
    PoleVectorSpace=(Type=Control,Name="None")
    PrimaryAxis=(X=1,Y=0,Z=0)
    SecondaryAxis=(X=0,Y=0,Z=1)
    SecondaryAxisWeight=1
    StretchMaximumRatio=1.25
    StretchStartRatio=0.75
    Weight=1
    bEnableStretch=False
    bPropagateToChildren=True
    PoleVector=&RigUnit_GetTransform(Item=(Type=Control,Name="dyn_spiderA_leg02_poleVector"), bInitial=False).Transform.Translation
    Effector=&RigUnit_GetTransform(Item=(Type=Bone,Name="dyn_spiderA_leg02_03"), bInitial=False).Transform
   )
   > &RigUnit_TwoBoneIKSimplePerItem(
    DebugSettings=(bEnabled=False,Scale=10,WorldOffset=(Rotation=(X=0,Y=0,Z=0,W=1),Translation=(X=0,Y=0,Z=0),Scale3D=(X=1,Y=1,Z=1)))
    EffectorItem=(Type=Bone,Name="dyn_spiderA_leg01_03")
    ItemA=(Type=Bone,Name="dynControlRig_spiderA_leg01_01")
    ItemALength=0
    ItemB=(Type=Bone,Name="dynControlRig_spiderA_leg01_02")
    ItemBLength=0
    PoleVectorKind=Location
    PoleVectorSpace=(Type=Control,Name="None")
    PrimaryAxis=(X=1,Y=0,Z=0)
    SecondaryAxis=(X=0,Y=0,Z=1)
    SecondaryAxisWeight=1
    StretchMaximumRatio=1.25
    StretchStartRatio=0.75
    Weight=1
    bEnableStretch=False
    bPropagateToChildren=True
    Effector=&RigUnit_GetTransform(Item=(Type=Bone,Name="dyn_spiderA_leg01_03"), bInitial=False).Transform
    PoleVector=&RigUnit_GetTransform(Item=(Type=Control,Name="dyn_spiderA_leg01_poleVector"), bInitial=False).Transform.Translation
   )
   &RigUnit_TwoBoneIKSimplePerItem(
    DebugSettings=(bEnabled=False,Scale=10,WorldOffset=(Rotation=(X=0,Y=0,Z=0,W=1),Translation=(X=0,Y=0,Z=0),Scale3D=(X=1,Y=1,Z=1)))
    EffectorItem=(Type=Bone,Name="dyn_spiderA_leg05_03")
    ItemA=(Type=Bone,Name="dyn_spiderA_leg05_01")
    ItemALength=0
    ItemB=(Type=Bone,Name="dyn_spiderA_leg05_02")
    ItemBLength=0
    PoleVectorKind=Location
    PoleVectorSpace=(Type=Control,Name="None")
    PrimaryAxis=(X=1,Y=0,Z=0)
    SecondaryAxis=(X=0,Y=0,Z=1)
    SecondaryAxisWeight=1
    StretchMaximumRatio=1.25
    StretchStartRatio=0.75
    Weight=1
    bEnableStretch=False
    bPropagateToChildren=True
    Effector=&RigUnit_GetTransform(Item=(Type=Bone,Name="dyn_spiderA_leg05_03"), bInitial=False).Transform
    PoleVector=&RigUnit_GetTransform(Item=(Type=Control,Name="dyn_spiderA_leg05_poleVector"), bInitial=False).Transform.Translation
    ExecutePin=@RigVMFunction_Sequence.E
   )
   > &RigUnit_TwoBoneIKSimplePerItem(
    DebugSettings=(bEnabled=False,Scale=10,WorldOffset=(Rotation=(X=0,Y=0,Z=0,W=1),Translation=(X=0,Y=0,Z=0),Scale3D=(X=1,Y=1,Z=1)))
    EffectorItem=(Type=Bone,Name="dyn_spiderA_leg06_03")
    ItemA=(Type=Bone,Name="dynControlRig_spiderA_leg06_01")
    ItemALength=0
    ItemB=(Type=Bone,Name="dynControlRig_spiderA_leg06_02")
    ItemBLength=0
    PoleVectorKind=Location
    PoleVectorSpace=(Type=Control,Name="None")
    PrimaryAxis=(X=1,Y=0,Z=0)
    SecondaryAxis=(X=0,Y=0,Z=1)
    SecondaryAxisWeight=1
    StretchMaximumRatio=1.25
    StretchStartRatio=0.75
    Weight=0.3
    bEnableStretch=False
    bPropagateToChildren=True
    Effector=&RigUnit_GetTransform(Item=(Type=Bone,Name="dyn_spiderA_leg06_03"), bInitial=False).Transform
    PoleVector=&RigUnit_GetTransform(Item=(Type=Control,Name="dyn_spiderA_leg06_poleVector"), bInitial=False).Transform.Translation
   )
   > &RigUnit_TwoBoneIKSimplePerItem(
    DebugSettings=(bEnabled=False,Scale=10,WorldOffset=(Rotation=(X=0,Y=0,Z=0,W=1),Translation=(X=0,Y=0,Z=0),Scale3D=(X=1,Y=1,Z=1)))
    EffectorItem=(Type=Bone,Name="dyn_spiderA_leg07_03")
    ItemA=(Type=Bone,Name="dynControlRig_spiderA_leg07_01")
    ItemALength=0
    ItemB=(Type=Bone,Name="dynControlRig_spiderA_leg07_02")
    ItemBLength=0
    PoleVectorKind=Location
    PoleVectorSpace=(Type=Control,Name="None")
    PrimaryAxis=(X=1,Y=0,Z=0)
    SecondaryAxis=(X=0,Y=0,Z=1)
    SecondaryAxisWeight=1
    StretchMaximumRatio=1.25
    StretchStartRatio=0.75
    Weight=0.5
    bEnableStretch=False
    bPropagateToChildren=True
    Effector=&RigUnit_GetTransform(Item=(Type=Bone,Name="dyn_spiderA_leg07_03"), bInitial=False).Transform
    PoleVector=&RigUnit_GetTransform(Item=(Type=Control,Name="dyn_spiderA_leg07_poleVector"), bInitial=False).Transform.Translation
   )
   > &RigUnit_TwoBoneIKSimplePerItem(
    DebugSettings=(bEnabled=False,Scale=10,WorldOffset=(Rotation=(X=0,Y=0,Z=0,W=1),Translation=(X=0,Y=0,Z=0),Scale3D=(X=1,Y=1,Z=1)))
    EffectorItem=(Type=Bone,Name="dyn_spiderA_leg08_03")
    ItemA=(Type=Bone,Name="dynControlRig_spiderA_leg08_01")
    ItemALength=0
    ItemB=(Type=Bone,Name="dynControlRig_spiderA_leg08_02")
    ItemBLength=0
    PoleVectorKind=Location
    PoleVectorSpace=(Type=Control,Name="None")
    PrimaryAxis=(X=1,Y=0,Z=0)
    SecondaryAxis=(X=0,Y=0,Z=1)
    SecondaryAxisWeight=1
    StretchMaximumRatio=1.25
    StretchStartRatio=0.75
    Weight=1
    bEnableStretch=False
    bPropagateToChildren=True
    PoleVector=&RigUnit_GetTransform(Item=(Type=Control,Name="dyn_spiderA_leg08_poleVector"), bInitial=False).Transform.Translation
    Effector=&RigUnit_GetTransform(Item=(Type=Bone,Name="dyn_spiderA_leg08_03"), bInitial=False).Transform
   )
   &RigUnit_GetTransform(Item=(Type=Control,Name="dyn_spiderA_leg08_IK_ctrl"), bInitial=False)
   &RigUnit_GetTransform(Item=(Type=Control,Name="dyn_spiderB_leg01_IK_ctrl"), bInitial=False)
   &RigUnit_GetTransform(Item=(Type=Control,Name="dyn_spiderB_leg08_IK_ctrl"), bInitial=False)
   &RigUnit_GetTransform(Item=(Type=Control,Name="dyn_spiderA_leg01_IK_ctrl"), bInitial=False)
   RigUnit_GetTransform = &RigUnit_GetTransform(Item=(Type=Control,Name="dyn_spiderB_main_ctrl"), bInitial=False)
   RigUnit_SetTransform = &RigUnit_SetTransform(ExecutePin=(), Item=(Type=Bone,Name="dyn_spiderB_main"), Weight=1, bInitial=False, Value=@RigUnit_GetTransform.Transform)
  ]}
]
hierarchy [
 root {loc=<0 0 0>, rot=<0 0 0 1>}
 pack {parent=root, loc=<0 -0.7316 0>, rot=<0 0 0 1>}
 base_spiderB_main {parent=pack, loc=<-10.5016 -9.2362 -5.3866>, rot=<0.4686 0.3032 0.6168 -0.5551>}
 dyn_spiderB_main {parent=base_spiderB_main, loc=<0 0 0>, rot=<0 0 0 1>}
 dynControlRig_spiderB_leg07_01 {parent=dyn_spiderB_main, loc=<-0.3955 0.2514 -0.7561>, rot=<-0.0964 -0.2407 0.7341 -0.6276>}
 dynControlRig_spiderB_leg07_02 {parent=dynControlRig_spiderB_leg07_01, loc=<2.0268 0 0>, rot=<0 -0.9101 0 -0.4144>}
 dynControlRig_spiderB_leg07_03 {parent=dynControlRig_spiderB_leg07_02, loc=<2.3731 0 0>, rot=<0 0 0 1>}
 dynControlRig_spiderB_leg06_01 {parent=dyn_spiderB_main, loc=<-0.0647 0.1012 -0.8166>, rot=<-0.4558 -0.196 0.6367 -0.5903>}
 dynControlRig_spiderB_leg06_02 {parent=dynControlRig_spiderB_leg06_01, loc=<2.0268 0 0>, rot=<0 -0.8948 0 -0.4465>}
 dynControlRig_spiderB_leg06_03 {parent=dynControlRig_spiderB_leg06_02, loc=<3.0647 0 0>, rot=<0 0 0 1>}
 dynControlRig_spiderB_leg05_01 {parent=dyn_spiderB_main, loc=<0.2264 0.0192 -0.7233>, rot=<-0.5932 0.1459 0.531 -0.5873>}
 dynControlRig_spiderB_leg05_02 {parent=dynControlRig_spiderB_leg05_01, loc=<2.1622 0 0>, rot=<0 -0.9007 0 -0.4344>}
 dynControlRig_spiderB_leg05_03 {parent=dynControlRig_spiderB_leg05_02, loc=<3.863 0 0>, rot=<0 0 0 1>}
 dynControlRig_spiderB_leg03_01 {parent=dyn_spiderB_main, loc=<-0.0714 0.1887 0.8123>, rot=<0.5661 -0.6207 0.2218 0.495>}
 dynControlRig_spiderB_leg03_02 {parent=dynControlRig_spiderB_leg03_01, loc=<2.0332 0 0>, rot=<0 -0.8957 0 -0.4446>}
 dynControlRig_spiderB_leg03_03 {parent=dynControlRig_spiderB_leg03_02, loc=<3.0086 0 0>, rot=<0 0 0 1>}
 dynControlRig_spiderB_leg02_01 {parent=dyn_spiderB_main, loc=<-0.3786 0.3302 0.7321>, rot=<0.6168 -0.7223 0.2814 0.137>}
 dynControlRig_spiderB_leg02_02 {parent=dynControlRig_spiderB_leg02_01, loc=<2.0961 0 0>, rot=<0 -0.9108 0 -0.4128>}
 dynControlRig_spiderB_leg02_03 {parent=dynControlRig_spiderB_leg02_02, loc=<2.4014 0 0>, rot=<0 0 0 1>}
 dynControlRig_spiderB_leg04_01 {parent=dyn_spiderB_main, loc=<0.2322 0.0878 0.723>, rot=<-0.559 0.5322 0.1432 -0.6195>}
 dynControlRig_spiderB_leg04_02 {parent=dynControlRig_spiderB_leg04_01, loc=<2.2846 0 0>, rot=<0 -0.9249 0 -0.3803>}
 dynControlRig_spiderB_leg04_03 {parent=dynControlRig_spiderB_leg04_02, loc=<3.8812 0 0>, rot=<0 0 0 1>}
 dynControlRig_spiderB_leg01_01 {parent=dyn_spiderB_main, loc=<-0.6403 0.284 0.5236>, rot=<0.4645 -0.7512 0.4607 0.0875>}
 dynControlRig_spiderB_leg01_02 {parent=dynControlRig_spiderB_leg01_01, loc=<1.9454 0 0>, rot=<0 -0.9306 0 -0.366>}
 dynControlRig_spiderB_leg01_03 {parent=dynControlRig_spiderB_leg01_02, loc=<2.8613 0 0>, rot=<0 0 0 1>}
 dynControlRig_spiderB_leg08_01 {parent=dyn_spiderB_main, loc=<-0.6453 0.2374 -0.5402>, rot=<-0.075 -0.4311 0.7746 -0.4567>}
 dynControlRig_spiderB_leg08_02 {parent=dynControlRig_spiderB_leg08_01, loc=<1.7689 0 0>, rot=<0 -0.9184 0 -0.3957>}
 dynControlRig_spiderB_leg08_03 {parent=dynControlRig_spiderB_leg08_02, loc=<2.6988 0 0>, rot=<0 0 0 1>}
 dyn_spiderB_leg01_01 {parent=base_spiderB_main, loc=<-0.6403 0.284 0.5236>, rot=<0.4645 -0.7512 0.4607 0.0875>}
 dyn_spiderB_leg01_02 {parent=dyn_spiderB_leg01_01, loc=<1.9454 0 0>, rot=<0 -0.9306 0 -0.366>}
 dyn_spiderB_leg01_03 {parent=dyn_spiderB_leg01_02, loc=<2.8613 0 0>, rot=<0 0 0 1>}
 dyn_spiderB_leg01_poleVector {controlType=EULER_TRANSFORM, parent=dyn_spiderB_leg01_03, type=Control, loc=<0 0 0>, rot=<0 0 0 1>}
 dyn_spiderB_leg02_01 {parent=base_spiderB_main, loc=<-0.3786 0.3302 0.7321>, rot=<0.6168 -0.7223 0.2814 0.137>}
 dyn_spiderB_leg02_02 {parent=dyn_spiderB_leg02_01, loc=<2.0961 0 0>, rot=<0 -0.9108 0 -0.4128>}
 dyn_spiderB_leg02_03 {parent=dyn_spiderB_leg02_02, loc=<2.4014 0 0>, rot=<0 0 0 1>}
 dyn_spiderB_leg02_poleVector {controlType=EULER_TRANSFORM, parent=dyn_spiderB_leg02_03, type=Control, loc=<0 0 0>, rot=<0 0 0 1>}
 dyn_spiderB_leg03_01 {parent=base_spiderB_main, loc=<-0.0714 0.1887 0.8123>, rot=<0.5661 -0.6207 0.2218 0.495>}
 dyn_spiderB_leg03_02 {parent=dyn_spiderB_leg03_01, loc=<2.0332 0 0>, rot=<0 -0.8957 0 -0.4446>}
 dyn_spiderB_leg03_03 {parent=dyn_spiderB_leg03_02, loc=<3.0086 0 0>, rot=<0 0 0 1>}
 dyn_spiderB_leg03_poleVector {controlType=EULER_TRANSFORM, parent=dyn_spiderB_leg03_03, type=Control, loc=<0 0 0>, rot=<0 0 0 1>}
 dyn_spiderB_leg04_01 {parent=base_spiderB_main, loc=<0.2322 0.0878 0.723>, rot=<-0.559 0.5322 0.1432 -0.6195>}
 dyn_spiderB_leg04_02 {parent=dyn_spiderB_leg04_01, loc=<2.2846 0 0>, rot=<0 -0.9249 0 -0.3803>}
 dyn_spiderB_leg04_03 {parent=dyn_spiderB_leg04_02, loc=<3.8812 0 0>, rot=<0 0 0 1>}
 dyn_spiderB_leg04_poleVector {controlType=EULER_TRANSFORM, parent=dyn_spiderB_leg04_03, type=Control, loc=<0 0 0>, rot=<0 0 0 1>}
 dyn_spiderB_leg05_01 {parent=base_spiderB_main, loc=<0.2264 0.0192 -0.7233>, rot=<-0.5932 0.1459 0.531 -0.5873>}
 dyn_spiderB_leg05_02 {parent=dyn_spiderB_leg05_01, loc=<2.1622 0 0>, rot=<0 -0.9007 0 -0.4344>}
 dyn_spiderB_leg05_03 {parent=dyn_spiderB_leg05_02, loc=<3.863 0 0>, rot=<0 0 0 1>}
 dyn_spiderB_leg05_poleVector {controlType=EULER_TRANSFORM, parent=dyn_spiderB_leg05_03, type=Control, loc=<0 0 0>, rot=<0 0 0 1>}
 dyn_spiderB_leg06_01 {parent=base_spiderB_main, loc=<-0.0647 0.1012 -0.8166>, rot=<-0.4558 -0.196 0.6367 -0.5903>}
 dyn_spiderB_leg06_02 {parent=dyn_spiderB_leg06_01, loc=<2.0268 0 0>, rot=<0 -0.8948 0 -0.4465>}
 dyn_spiderB_leg06_03 {parent=dyn_spiderB_leg06_02, loc=<3.0647 0 0>, rot=<0 0 0 1>}
 dyn_spiderB_leg06_poleVector {controlType=EULER_TRANSFORM, parent=dyn_spiderB_leg06_03, type=Control, loc=<0 0 0>, rot=<0 0 0 1>}
 dyn_spiderB_leg07_01 {parent=base_spiderB_main, loc=<-0.3955 0.2514 -0.7561>, rot=<-0.0964 -0.2407 0.7341 -0.6276>}
 dyn_spiderB_leg07_02 {parent=dyn_spiderB_leg07_01, loc=<2.0268 0 0>, rot=<0 -0.9101 0 -0.4144>}
 dyn_spiderB_leg07_03 {parent=dyn_spiderB_leg07_02, loc=<2.3731 0 0>, rot=<0 0 0 1>}
 dyn_spiderB_leg07_poleVector {controlType=EULER_TRANSFORM, parent=dyn_spiderB_leg07_03, type=Control, loc=<0 0 0>, rot=<0 0 0 -1>}
 dyn_spiderB_leg08_01 {parent=base_spiderB_main, loc=<-0.6453 0.2374 -0.5402>, rot=<-0.075 -0.4311 0.7746 -0.4567>}
 dyn_spiderB_leg08_02 {parent=dyn_spiderB_leg08_01, loc=<1.7689 0 0>, rot=<0 -0.9184 0 -0.3957>}
 dyn_spiderB_leg08_03 {parent=dyn_spiderB_leg08_02, loc=<2.6988 0 0>, rot=<0 0 0 1>}
 dyn_spiderB_leg08_poleVector {controlType=EULER_TRANSFORM, parent=dyn_spiderB_leg08_03, type=Control, loc=<0 0 0>, rot=<0 0 0 -1>}
 base_spiderA_main {parent=pack, loc=<3.5423 -19.8003 5.3072>, rot=<0.3523 -0.836 0.4197 -0.0282>}
 dyn_spiderA_main {parent=base_spiderA_main, loc=<0 0 0>, rot=<0 0 0 1>}
 dynControlRig_spiderA_leg05_01 {parent=dyn_spiderA_main, loc=<0.2264 0.0192 -0.7233>, rot=<-0.5932 0.1459 0.531 -0.5873>}
 dynControlRig_spiderA_leg05_02 {parent=dynControlRig_spiderA_leg05_01, loc=<2.1622 0 0>, rot=<0 -0.9007 0 -0.4344>}
 dynControlRig_spiderA_leg05_03 {parent=dynControlRig_spiderA_leg05_02, loc=<3.863 0 0>, rot=<0 0 0 1>}
 dynControlRig_spiderA_leg04_01 {parent=dyn_spiderA_main, loc=<0.2322 0.0878 0.723>, rot=<-0.559 0.5322 0.1432 -0.6195>}
 dynControlRig_spiderA_leg04_02 {parent=dynControlRig_spiderA_leg04_01, loc=<2.2846 0 0>, rot=<0 -0.9249 0 -0.3803>}
 dynControlRig_spiderA_leg04_03 {parent=dynControlRig_spiderA_leg04_02, loc=<3.8812 0 0>, rot=<0 0 0 1>}
 dynControlRig_spiderA_leg03_01 {parent=dyn_spiderA_main, loc=<-0.0714 0.1887 0.8123>, rot=<0.5661 -0.6207 0.2218 0.495>}
 dynControlRig_spiderA_leg03_02 {parent=dynControlRig_spiderA_leg03_01, loc=<2.0332 0 0>, rot=<0 -0.8957 0 -0.4446>}
 dynControlRig_spiderA_leg03_03 {parent=dynControlRig_spiderA_leg03_02, loc=<3.0086 0 0>, rot=<0 0 0 1>}
 dynControlRig_spiderA_leg02_01 {parent=dyn_spiderA_main, loc=<-0.3786 0.3302 0.7321>, rot=<0.6168 -0.7223 0.2814 0.137>}
 dynControlRig_spiderA_leg02_02 {parent=dynControlRig_spiderA_leg02_01, loc=<2.0961 0 0>, rot=<0 -0.9108 0 -0.4128>}
 dynControlRig_spiderA_leg02_03 {parent=dynControlRig_spiderA_leg02_02, loc=<2.4014 0 0>, rot=<0 0 0 1>}
 dynControlRig_spiderA_leg06_01 {parent=dyn_spiderA_main, loc=<-0.0647 0.1012 -0.8166>, rot=<-0.4558 -0.196 0.6367 -0.5903>}
 dynControlRig_spiderA_leg06_02 {parent=dynControlRig_spiderA_leg06_01, loc=<2.0268 0 0>, rot=<0 -0.8948 0 -0.4465>}
 dynControlRig_spiderA_leg06_03 {parent=dynControlRig_spiderA_leg06_02, loc=<3.0647 0 0>, rot=<0 0 0 1>}
 dynControlRig_spiderA_leg07_01 {parent=dyn_spiderA_main, loc=<-0.3955 0.2514 -0.7561>, rot=<-0.1648 -0.3345 0.7125 -0.5944>}
 dynControlRig_spiderA_leg07_02 {parent=dynControlRig_spiderA_leg07_01, loc=<2.0268 0 0>, rot=<0 -0.9368 0 -0.3498>}
 dynControlRig_spiderA_leg07_03 {parent=dynControlRig_spiderA_leg07_02, loc=<2.3731 0 0>, rot=<0 0 0 1>}
 dynControlRig_spiderA_leg08_01 {parent=dyn_spiderA_main, loc=<-0.6453 0.2374 -0.5402>, rot=<-0.075 -0.4311 0.7746 -0.4567>}
 dynControlRig_spiderA_leg08_02 {parent=dynControlRig_spiderA_leg08_01, loc=<1.7689 0 0>, rot=<0 -0.9184 0 -0.3957>}
 dynControlRig_spiderA_leg08_03 {parent=dynControlRig_spiderA_leg08_02, loc=<2.6988 0 0>, rot=<0 0 0 1>}
 dynControlRig_spiderA_leg01_01 {parent=dyn_spiderA_main, loc=<-0.6403 0.284 0.5236>, rot=<0.4645 -0.7512 0.4607 0.0875>}
 dynControlRig_spiderA_leg01_02 {parent=dynControlRig_spiderA_leg01_01, loc=<1.9454 0 0>, rot=<0 -0.9306 0 -0.366>}
 dynControlRig_spiderA_leg01_03 {parent=dynControlRig_spiderA_leg01_02, loc=<2.8613 0 0>, rot=<0 0 0 1>}
 dyn_spiderA_leg01_01 {parent=base_spiderA_main, loc=<-0.6403 0.284 0.5236>, rot=<0.4645 -0.7512 0.4607 0.0875>}
 dyn_spiderA_leg01_02 {parent=dyn_spiderA_leg01_01, loc=<1.9454 0 0>, rot=<0 -0.9306 0 -0.366>}
 dyn_spiderA_leg01_03 {parent=dyn_spiderA_leg01_02, loc=<2.8613 0 0>, rot=<0 0 0 1>}
 dyn_spiderA_leg01_poleVector {controlType=EULER_TRANSFORM, parent=dyn_spiderA_leg01_03, type=Control, loc=<0 0 0>, rot=<0 0 0 1>}
 dyn_spiderA_leg02_01 {parent=base_spiderA_main, loc=<-0.3786 0.3302 0.7321>, rot=<0.6168 -0.7223 0.2814 0.137>}
 dyn_spiderA_leg02_02 {parent=dyn_spiderA_leg02_01, loc=<2.0961 0 0>, rot=<0 -0.9108 0 -0.4128>}
 dyn_spiderA_leg02_03 {parent=dyn_spiderA_leg02_02, loc=<2.4014 0 0>, rot=<0 0 0 1>}
 dyn_spiderA_leg02_poleVector {controlType=EULER_TRANSFORM, parent=dyn_spiderA_leg02_03, type=Control, loc=<0 0 0>, rot=<0 0 0 -1>}
 dyn_spiderA_leg03_01 {parent=base_spiderA_main, loc=<-0.0714 0.1887 0.8123>, rot=<0.5661 -0.6207 0.2218 0.495>}
 dyn_spiderA_leg03_02 {parent=dyn_spiderA_leg03_01, loc=<2.0332 0 0>, rot=<0 -0.8957 0 -0.4446>}
 dyn_spiderA_leg03_03 {parent=dyn_spiderA_leg03_02, loc=<3.0086 0 0>, rot=<0 0 0 1>}
 dyn_spiderA_leg03_poleVector {controlType=EULER_TRANSFORM, parent=dyn_spiderA_leg03_03, type=Control, loc=<0 0 0>, rot=<0 0 0 -1>}
 dyn_spiderA_leg04_01 {parent=base_spiderA_main, loc=<0.2322 0.0878 0.723>, rot=<-0.559 0.5322 0.1432 -0.6195>}
 dyn_spiderA_leg04_02 {parent=dyn_spiderA_leg04_01, loc=<2.2846 0 0>, rot=<0 -0.9249 0 -0.3803>}
 dyn_spiderA_leg04_03 {parent=dyn_spiderA_leg04_02, loc=<3.8812 0 0>, rot=<0 0 0 1>}
 dyn_spiderA_leg04_poleVector {controlType=EULER_TRANSFORM, parent=dyn_spiderA_leg04_03, type=Control, loc=<0 0 0>, rot=<0 0 0 1>}
 dyn_spiderA_leg05_01 {parent=base_spiderA_main, loc=<0.2264 0.0192 -0.7233>, rot=<-0.5932 0.1459 0.531 -0.5873>}
 dyn_spiderA_leg05_02 {parent=dyn_spiderA_leg05_01, loc=<2.1622 0 0>, rot=<0 -0.9007 0 -0.4344>}
 dyn_spiderA_leg05_03 {parent=dyn_spiderA_leg05_02, loc=<3.863 0 0>, rot=<0 0 0 1>}
 dyn_spiderA_leg05_poleVector {controlType=EULER_TRANSFORM, parent=dyn_spiderA_leg05_03, type=Control, loc=<0 0 0>, rot=<0 0 0 1>}
 dyn_spiderA_leg06_01 {parent=base_spiderA_main, loc=<-0.0647 0.1012 -0.8166>, rot=<-0.4558 -0.196 0.6367 -0.5903>}
 dyn_spiderA_leg06_02 {parent=dyn_spiderA_leg06_01, loc=<2.0268 0 0>, rot=<0 -0.8948 0 -0.4465>}
 dyn_spiderA_leg06_03 {parent=dyn_spiderA_leg06_02, loc=<3.0647 0 0>, rot=<0 0 0 1>}
 dyn_spiderA_leg06_poleVector {controlType=EULER_TRANSFORM, parent=dyn_spiderA_leg06_03, type=Control, loc=<0 0 0>, rot=<0 0 0 1>}
 dyn_spiderA_leg07_01 {parent=base_spiderA_main, loc=<-0.3955 0.2514 -0.7561>, rot=<-0.1648 -0.3345 0.7125 -0.5944>}
 dyn_spiderA_leg07_02 {parent=dyn_spiderA_leg07_01, loc=<2.0268 0 0>, rot=<0 -0.9368 0 -0.3498>}
 dyn_spiderA_leg07_03 {parent=dyn_spiderA_leg07_02, loc=<2.3731 0 0>, rot=<0 0 0 1>}
 dyn_spiderA_leg07_poleVector {controlType=EULER_TRANSFORM, parent=dyn_spiderA_leg07_03, type=Control, loc=<0 0 0>, rot=<0 0 0 1>}
 dyn_spiderA_leg08_01 {parent=base_spiderA_main, loc=<-0.6453 0.2374 -0.5402>, rot=<-0.075 -0.4311 0.7746 -0.4567>}
 dyn_spiderA_leg08_02 {parent=dyn_spiderA_leg08_01, loc=<1.7689 0 0>, rot=<0 -0.9184 0 -0.3957>}
 dyn_spiderA_leg08_03 {parent=dyn_spiderA_leg08_02, loc=<2.6988 0 0>, rot=<0 0 0 1>}
 dyn_spiderA_leg08_poleVector {controlType=EULER_TRANSFORM, parent=dyn_spiderA_leg08_03, type=Control, loc=<0 0 0>, rot=<0 0 0 1>}
 SpiderB {type=Null, loc=<0 0 0>, rot=<0 0 0 1>}
 dyn_spiderB_main_ctrl {controlType=EULER_TRANSFORM, parent=SpiderB, type=Control, loc=<-10.5016 -9.9678 -5.3866>, rot=<-0.4686 -0.3032 -0.6168 0.5551>}
 dyn_spiderB_leg04_IK_ctrl {controlType=EULER_TRANSFORM, parent=SpiderB, type=Control, loc=<-8.2631 -10.3656 -3.1994>, rot=<-0.0613 -0.0111 -0.2612 0.9633>}
 dyn_spiderB_leg03_IK_ctrl {controlType=EULER_TRANSFORM, parent=SpiderB, type=Control, loc=<-7.9537 -8.9426 -4.187>, rot=<-0.4147 0.1087 -0.2048 0.8799>}
 dyn_spiderB_leg02_IK_ctrl {controlType=EULER_TRANSFORM, parent=SpiderB, type=Control, loc=<-9.0262 -7.6541 -5.2843>, rot=<-0.6482 0.0059 0.0164 0.7613>}
 dyn_spiderB_leg01_IK_ctrl {controlType=EULER_TRANSFORM, parent=SpiderB, type=Control, loc=<-8.3694 -8.5195 -6.1228>, rot=<-0.7877 0.112 -0.1052 0.5966>}
 dyn_spiderB_leg08_IK_ctrl {controlType=EULER_TRANSFORM, parent=SpiderB, type=Control, loc=<-9.2948 -11.3845 -7.3278>, rot=<0.9481 -0.1757 -0.0509 0.2601>}
 dyn_spiderB_leg07_IK_ctrl {controlType=EULER_TRANSFORM, parent=SpiderB, type=Control, loc=<-10.4821 -12.0435 -7.1426>, rot=<0.8532 -0.1782 -0.2451 0.4245>}
 dyn_spiderB_leg06_IK_ctrl {controlType=EULER_TRANSFORM, parent=SpiderB, type=Control, loc=<-9.1253 -12.653 -5.7591>, rot=<0.7301 -0.1041 -0.0399 0.6741>}
 dyn_spiderB_leg05_IK_ctrl {controlType=EULER_TRANSFORM, parent=SpiderB, type=Control, loc=<-8.9005 -12.5936 -3.8416>, rot=<0.4225 -0.1704 -0.0649 0.8878>}
 SpiderA {type=Null, loc=<0 0 0>, rot=<0 0 0 1>}
 dyn_spiderA_leg08_IK_ctrl {controlType=EULER_TRANSFORM, parent=SpiderA, type=Control, loc=<2.8317 -17.9748 4.8737>, rot=<-0.1456 0.558 0.7437 0.338>}
 dyn_spiderA_leg07_IK_ctrl {controlType=EULER_TRANSFORM, parent=SpiderA, type=Control, loc=<2.8708 -18.0681 6.2405>, rot=<-0.1061 0.3203 0.8516 0.4011>}
 dyn_spiderA_leg06_IK_ctrl {controlType=EULER_TRANSFORM, parent=SpiderA, type=Control, loc=<1.0884 -18.7531 5.5474>, rot=<-0.3524 0.19 0.8969 0.188>}
 dyn_spiderA_leg05_IK_ctrl {controlType=EULER_TRANSFORM, parent=SpiderA, type=Control, loc=<0.1078 -20.4164 5.4977>, rot=<-0.3338 -0.1344 0.933 -0.0083>}
 dyn_spiderA_leg04_IK_ctrl {controlType=EULER_TRANSFORM, parent=SpiderA, type=Control, loc=<1.0846 -21.8487 3.8311>, rot=<-0.2955 -0.6371 0.7114 -0.0264>}
 dyn_spiderA_leg03_IK_ctrl {controlType=EULER_TRANSFORM, parent=SpiderA, type=Control, loc=<2.421 -21.5576 2.7241>, rot=<-0.3047 -0.8447 0.412 -0.1548>}
 dyn_spiderA_leg02_IK_ctrl {controlType=EULER_TRANSFORM, parent=SpiderA, type=Control, loc=<4.4124 -21.3712 2.8417>, rot=<-0.2423 -0.8319 0.2182 -0.4489>}
 dyn_spiderA_leg01_IK_ctrl {controlType=EULER_TRANSFORM, parent=SpiderA, type=Control, loc=<3.8544 -20.1285 2.6755>, rot=<0.1563 0.9261 -0.0055 0.3434>}
]
