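# Physics asset for the F_MED_PhoneCharger_Pack backpack cosmetic, giving collision bodies and joints to the pack and two "spider" attachments.
# Two larger capsules cover the pack bone, and small capsules cover each attachment's base and dyn bones.
# Each base bone has swing and twist locked to the pack.
# Each dyn bone hangs off its base through a joint with narrow angular limit values, apparently for secondary motion.
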PhysicsAsset /BRCosmetics/Accessories/FORT_Backpacks/Backpack_F_MED_PhoneCharger/Meshes/F_MED_PhoneCharger_Pack_Physics
bodies {
 base_spiderA_main [{shapeType=Capsule, length=1, radius=0.5}]
 base_spiderB_main [{shapeType=Capsule, length=1, radius=0.5}]
 dyn_spiderA_main [{shapeType=Capsule, length=1, radius=0.5, center=<-0.5142 0 0>, rotation=<rot 90 0 0>}]
 dyn_spiderB_main [{shapeType=Capsule, length=1, radius=1, center=<-0.9393 0 0>, rotation=<rot 90 0 0>}]
 pack [
  {shapeType=Capsule, length=1, radius=5.5, center=<-4.221 -10.4139 -6.2818>}
  {shapeType=Capsule, length=1, radius=4.5, center=<1.3304 -17.8163 0.6844>}
 ]
}
constraints [
 {bone1Name=dyn_spiderB_main, bone2Name=base_spiderB_main, swing1LimitDegrees=10, swing2LimitDegrees=10, twistLimitDegrees=10}
 {bone1Name=dyn_spiderA_main, bone2Name=base_spiderA_main, swing1LimitDegrees=10, swing2LimitDegrees=10, twistLimitDegrees=10}
 {
  bone1Name=base_spiderA_main
  bone2Name=pack
  swing1LimitDegrees=45
  swing1Motion=Locked
  swing2LimitDegrees=45
  swing2Motion=Locked
  twistLimitDegrees=45
  twistMotion=Locked
 }
 {
  bone1Name=base_spiderB_main
  bone2Name=pack
  swing1LimitDegrees=45
  swing1Motion=Locked
  swing2LimitDegrees=45
  swing2Motion=Locked
  twistLimitDegrees=45
  twistMotion=Locked
 }
]
