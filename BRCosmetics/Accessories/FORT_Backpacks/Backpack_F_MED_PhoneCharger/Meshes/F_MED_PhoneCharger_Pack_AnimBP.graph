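# Animation Blueprint for the PhoneCharger backpack cosmetic, playing random idle and idle-break animations and swaying two "spider" bones with player movement.
# The event graph thresholds movement into directional flags, interpolates them into offsets, and gives spider B a scaled-down copy.
# Additive world-space bone transforms apply these offsets before a Control Rig produces the final pose.
# A RigidBody node is unconnected, and the deprecated gravity override's result is unused.
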AnimBlueprint /BRCosmetics/Accessories/FORT_Backpacks/Backpack_F_MED_PhoneCharger/Meshes/F_MED_PhoneCharger_Pack_AnimBP
target_skeleton=&/F_MED_PhoneCharger_Pack_Skeleton
graphs [
 AnimGraph {nodes [
   RigidBody!(ComponentPose, Alpha, OverrideWorldGravity:Vector="0,0,0") -> (Pose) {
    bOverrideWorldGravity=true
    overridePhysicsAsset=&/F_MED_PhoneCharger_Pack_Physics
    alphaBoolBlend {blendOption=Linear}
    componentLinearAccScale=<0.2 0.1 0.1>
    componentLinearVelScale=<0.2 0.1 0.1>
    simSpaceSettings {velocityScaleZ=0, worldAlpha=0.8}
   }
   AnimGraphNode_StateMachine_0 = Backpack States() {is=title AnimGraphNode_StateMachine}
   n1 = "Transform (Modify) Bone - Bone: dyn_spiderA_main"!(ComponentPose=Local To Component(LocalPose=@AnimGraphNode_StateMachine_0), Alpha=0.6, Translation=Get SpiderA_Pos() -> (SpiderA_Pos:Vector)) -> (Pose) {translationMode=BMM_Additive, translationSpace=BCS_WorldSpace, alphaBoolBlend {blendOption=Linear}, boneToModify {boneName=dyn_spiderA_main}}
   n2 = "Transform (Modify) Bone - Bone: dyn_spiderB_main"!(ComponentPose=@n1, Alpha=0.6, Translation=Get SpiderB_Pos() -> (SpiderB_Pos:Vector)) -> (Pose) {translationMode=BMM_Additive, translationSpace=BCS_WorldSpace, alphaBoolBlend {blendOption=Linear}, boneToModify {boneName=dyn_spiderB_main}}
   n3 = Control Rig(Source=Component To Local(ComponentPose=@n2)) {
    alphaBoolBlend {blendOption=Linear}
    controlRigAssetReference {blueprintRigClass=&/F_MED_PhoneCharger_Pack_CtrlRig.F_MED_PhoneCharger_Pack_CtrlRig_C}
   }
   Output Pose(Result=@n3) {layerGroup=None}
  ]}
 Backpack States {
  entry=Idle
  path=AnimGraph.AnimGraphNode_StateMachine_0.Backpack States
  ids=id
  nodes [AnimStateNode_0 = Idle() {
    stateEntered {guid=#1, notifyName=Entered_RandomLocoBreak}
    stateFullyBlended {guid=#2}
    stateLeft {guid=#3}
   }]
 }
 Idle {
  path=AnimGraph.AnimGraphNode_StateMachine_0.Backpack States.AnimStateNode_0.Idle
  nodes [
   n1 = Random Sequence Player!() -> (Pose:Pose Link Structure) {
    is=title AnimGraphNode_RandomPlayer
    entries [
     {
      chanceToPlay=0.7
      maxLoopCount=0
      maxPlayRate=1
      minLoopCount=0
      minPlayRate=1
      sequence=/BRCosmetics/Accessories/FORT_Backpacks/Backpack_F_MED_PhoneCharger/Anim/Backpack_PhoneCharger_Idle
      blendIn {blendOption=Linear, blendTime=0.2, customCurve=null}
     }
     {
      chanceToPlay=0.5
      maxLoopCount=0
      maxPlayRate=1
      minLoopCount=0
      minPlayRate=1
      sequence=/BRCosmetics/Accessories/FORT_Backpacks/Backpack_F_MED_PhoneCharger/Anim/Backpack_PhoneCharger_IdleBreak1
      blendIn {blendOption=Linear, blendTime=0.2, customCurve=null}
     }
     {
      chanceToPlay=0.4
      maxLoopCount=0
      maxPlayRate=1
      minLoopCount=0
      minPlayRate=1
      sequence=/BRCosmetics/Accessories/FORT_Backpacks/Backpack_F_MED_PhoneCharger/Anim/Backpack_PhoneCharger_IdleBreak2
      blendIn {blendOption=Linear, blendTime=0.2, customCurve=null}
     }
    ]
   }
   Output Animation Pose(Result=@n1)
  ]
 }
 EventGraph {nodes [
   n1 = Event BlueprintUpdateAnimation()
   > n2 = Sequence!(execute) -> (then_0, then_1, then_2:Exec)
   > n3 = Gravity Override State Function DEPRECATED(CharacterPartAnimInstance=Self-Reference(), GameTimeInSeconds=GetGameTimeInSeconds(), GravityOverrideParams=Get Gravity Override Params())
   > Set GravityOverride(GravityOverride=@n3)
   n4 = PlayerMovement(execute=@n2.then_1, AnimInstance=Self-Reference(), XDotProductCoefficient=100, YDotProductCoefficient=100, ZDotProductCoefficient=100)
   n5 = BreakVector(InVec=MakeVector(X=@n4, Y=@n4.YMovement, Z=@n4.ZMovement))
   > Set Left(Left:double="To Float (Boolean)"(InBool="float < float"(A=@n5, B=-50))) -> (Output_Get:double)
   > Set Right(Right:double="To Float (Boolean)"(InBool="float > float"(A=@n5, B=50))) -> (Output_Get:double)
   > Set Forward(Forward:double="To Float (Boolean)"(InBool="float > float"(A=@n5.Y, B=40))) -> (Output_Get:double)
   > Set Backward(Backward:double="To Float (Boolean)"(InBool="float < float"(A=@n5.Y, B=-40))) -> (Output_Get:double)
   > Set Downwards(Downwards:double="To Float (Boolean)"(InBool="float < float"(A=@n5.Z, B=-40))) -> (Output_Get:double)
   > Set Upwards(Upwards:double="To Float (Boolean)"(InBool="float > float"(A=@n5.Z, B=50))) -> (Output_Get:double)
   n6 = float * float(A=Get Downwards() -> (Downwards:double), B=2) {is=title K2Node_PromotableOperator}
   n7 = float * float(A=Get Upwards() -> (Upwards:double), B=-0.3)
   n8 = Set Spiders_Z_Float(
    execute=@n2.then_2
    Spiders_Z_Float:double=FInterpTo(Current=Get Spiders_Z_Float() -> (Spiders_Z_Float:double), Target=float + float(A=@n6, B=@n7), DeltaTime=@n1.DeltaTimeX, InterpSpeed=8)
   ) -> (Output_Get:double)
   n9 = float * float(A=Get Right() -> (Right:double), B=-2)
   n10 = float * float(A=Get Left() -> (Left:double), B=1)
   > n11 = Set Spiders_X_Float(Spiders_X_Float:double=FInterpTo(Current=Get Spiders_X_Float() -> (Spiders_X_Float:double), Target=float + float(A=@n9, B=@n10), DeltaTime=@n1.DeltaTimeX, InterpSpeed=8)) -> (Output_Get:double)
   n12 = float * float(A=Get Forward() -> (Forward:double), B=-1)
   n13 = float * float(A=Get Backward() -> (Backward:double), B=1)
   > n14 = Set Spiders_Y_Float(Spiders_Y_Float:double=FInterpTo(Current=Get Spiders_Y_Float() -> (Spiders_Y_Float:double), Target=float + float(A=@n13, B=@n12), DeltaTime=@n1.DeltaTimeX, InterpSpeed=8)) -> (Output_Get:double)
   n15 = MakeVector(X=@n11, Y=@n14, Z=@n8)
   > Set SpiderA_Pos(SpiderA_Pos:Vector=@n15) -> (Output_Get:Vector)
   n16 = BreakVector(InVec=@n15)
   > Set SpiderB_Pos(SpiderB_Pos:Vector=MakeVector(X=float * float(A=@n16, B=0.5), Y=float * float(A=@n16.Y, B=0.5), Z=float * float(A=@n16.Z, B=0.8))) -> (Output_Get:Vector)
   TryGetPawnOwner()
   FInterpTo(Current=0, Target=0, DeltaTime=0, InterpSpeed=0)
   Ease!(Function:EEasingFunc Enum=EaseInOut, Alpha:double, A:Wildcard, B:Wildcard, BlendExp:double=2) -> (Result:Wildcard) {is=title K2Node_EaseFunction}
  ]}
]
variables [
 Backward {type=number}
 Downwards {type=number}
 Forward {type=number}
 Gravity Override Params {type=GravityOverrideParamsStruct}
 GravityOverride {type=Vector}
 Left {type=number}
 NewVar {type=number}
 Right {type=number}
 SpiderA_Pos {type=Vector}
 SpiderB_Pos {type=Vector}
 Spiders_X_Float {type=number}
 Spiders_Y_Float {type=number}
 Spiders_Z_Float {type=number}
 Upwards {type=number}
]
